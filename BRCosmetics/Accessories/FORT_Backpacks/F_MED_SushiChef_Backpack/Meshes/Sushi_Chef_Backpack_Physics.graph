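# Physics asset for the Sushi Chef backpack cosmetic in BRCosmetics, giving its zipper and tail capsule bodies for physical motion.
# The tail chains from the root through a base and main segment into left and right tips, with tight angular limits on the tips.
# The base-to-root joint swings and twists freely, while the zipper hangs directly off the root on its own constraint.
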PhysicsAsset /BRCosmetics/Accessories/FORT_Backpacks/F_MED_SushiChef_Backpack/Meshes/Sushi_Chef_Backpack_Physics
bodies {
 backpack_root [{shapeType=Capsule, length=6, radius=7.75, center=<-5.7467 0.2112 0.001>, rotation=<rot 90 0 -3.9297>}]
 dyn_zipper [{shapeType=Capsule, length=1.7694, radius=3.6726, center=<0 0 -2.412>, rotation=<rot -14.3389 0 0>}]
 tail_01_l [{shapeType=Capsule, length=2.6774, radius=1.6021, rotation=<rot 90 0 0>}]
 tail_01_r [{shapeType=Capsule, length=2.5989, radius=1.8366, rotation=<rot 90 0 0>}]
 tail_base [{shapeType=Capsule, length=0.505, radius=0.505, rotation=<rot 0 -90 0>}]
 tail_main [{shapeType=Capsule, length=6.0563, radius=3.2392, center=<-1.0935 -0.0879 0>}]
}
constraints [
 {bone1Name=tail_01_r, bone2Name=tail_main, swing1LimitDegrees=20, swing2LimitDegrees=20, twistLimitDegrees=5}
 {bone1Name=tail_01_l, bone2Name=tail_main, swing1LimitDegrees=20, swing2LimitDegrees=20, twistLimitDegrees=5}
 {bone1Name=tail_main, bone2Name=tail_base, swing1LimitDegrees=45, swing2LimitDegrees=80, twistLimitDegrees=6}
 {
  bone1Name=tail_base
  bone2Name=backpack_root
  swing1LimitDegrees=45
  swing1Motion=Free
  swing2LimitDegrees=45
  swing2Motion=Free
  twistLimitDegrees=45
  twistMotion=Free
 }
 {bone1Name=dyn_zipper, bone2Name=backpack_root, swing1LimitDegrees=10, swing2LimitDegrees=45, twistLimitDegrees=45}
]
