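# Animation Blueprint for the Sushi Chef backpack that adds secondary physics motion and tilts its tail base with body pose and skydiving.
# The input pose passes through a partially weighted RigidBody node, then two additive tail_base rotations.
# One rotation maps the owner's spine bend to roll; the other is an interpolated skydiving tilt, reduced in dive mode.
# An Anim Dynamics tail chain and a sequence player are unconnected to the output.
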AnimBlueprint /BRCosmetics/Accessories/FORT_Backpacks/F_MED_SushiChef_Backpack/Meshes/Sushi_Chef_Backpack_AnimBP
target_skeleton=/Game/Accessories/FORT_Backpacks/F_MED_SushiChef_Backpack/Meshes/Sushi_Chef_Backpack_Skeleton
graphs [
 AnimGraph {nodes [
   n1 = Anim Dynamics - Bone: tail_main(Alpha=0.375, ExternalForce="0,0,0") {
    angularDampingOverride=0.999
    angularSpringConstant=30000
    bAngularSpring=true
    bGravityOverrideInSimSpace=true
    bOverrideAngularDamping=true
    gravityScale=0.3
    simulationSpace=BoneRelative
    alphaBoolBlend {blendOption=Linear}
    boundBone {boneName=tail_main}
    componentAppliedLinearAccClamp=<1000 1000 1000>
    componentLinearAccScale=<1 1 1>
    componentLinearVelScale=<1 1 1>
    physicsBodyDefinitions [{
      sphereCollisionRadius=0
      boundBone {boneName=tail_main}
      boxExtents=<50 80 80>
      constraintSetup {angularLimitsMax=<0 30 10>, angularLimitsMin=<0 -30 -30>, angularTarget=<1 0 0>}
      localJointOffset=<-45 0 0.25>
     }]
   }
   n2 = Anim Dynamics - Bone: tail_01_r(ComponentPose=@n1, Alpha=1, ExternalForce="0,0,0") {
    angularDampingOverride=0.998
    angularSpringConstant=5000
    bAngularSpring=true
    bGravityOverrideInSimSpace=true
    bOverrideAngularDamping=true
    gravityScale=0.375
    simulationSpace=BoneRelative
    alphaBoolBlend {blendOption=Linear}
    boundBone {boneName=tail_01_r}
    componentAppliedLinearAccClamp=<1000 1000 1000>
    componentLinearAccScale=<1 1 1>
    componentLinearVelScale=<1 1 1>
    physicsBodyDefinitions [{
      sphereCollisionRadius=0
      boundBone {boneName=tail_01_r}
      boxExtents=<35 60 60>
      constraintSetup {angularLimitsMax=<0 50 10>, angularLimitsMin=<0 -50 -30>, angularTarget=<1 0 0>}
      localJointOffset=<-30 0 0>
     }]
   }
   Anim Dynamics - Bone: tail_01_l(ComponentPose=@n2, Alpha=1, ExternalForce="0,0,0") {
    angularDampingOverride=0.998
    angularSpringConstant=6000
    bAngularSpring=true
    bGravityOverrideInSimSpace=true
    bOverrideAngularDamping=true
    gravityScale=0.31
    simulationSpace=BoneRelative
    alphaBoolBlend {blendOption=Linear}
    boundBone {boneName=tail_01_l}
    componentAppliedLinearAccClamp=<1000 1000 1000>
    componentLinearAccScale=<1 1 1>
    componentLinearVelScale=<1 1 1>
    physicsBodyDefinitions [{
      sphereCollisionRadius=0
      boundBone {boneName=tail_01_l}
      boxExtents=<35 60 60>
      constraintSetup {angularLimitsMax=<0 50 10>, angularLimitsMin=<0 -50 -30>, angularTarget=<1 0 0>}
      localJointOffset=<-30 0 0>
     }]
   }
   AnimGraphNode_SubInput_0 = Input Pose()
   n3 = RigidBody(ComponentPose=Local To Component(LocalPose=@AnimGraphNode_SubInput_0), Alpha=0.35) {
    bOverrideWorldGravity=true
    overridePhysicsAsset=&/Sushi_Chef_Backpack_Physics
    simulationSpace=BaseBoneSpace
    alphaBoolBlend {blendOption=Linear}
    baseBoneRef {boneName=backpack_root}
    componentAppliedLinearAccClamp=<1000 1000 1000>
    componentLinearAccScale=<1 1 1>
    componentLinearVelScale=<1 1 1>
    overrideWorldGravity {z=-980}
   }
   n4 = "Transform (Modify) Bone - Bone: tail_base"(
    ComponentPose="Transform (Modify) Bone - Bone: tail_base"(ComponentPose=@n3, Alpha=1, Translation="0,0,0", Rotation=Get backpack_rotate(), Scale="1,1,1") {rotationMode=BMM_Additive, alphaBoolBlend {blendOption=Linear}, boneToModify {boneName=tail_base}}
    Alpha=Get float_var() -> (float_var:double)
    Translation="0,0,0"
    Rotation="0,0,35"
    Scale="1,1,1"
   ) {
    rotationMode=BMM_Additive
    alphaBoolBlend {blendOption=Linear}
    alphaScaleBiasClamp {bInterpResult=true}
    boneToModify {boneName=tail_base}
    rotation {roll=35}
   }
   Output Pose(Result=Component To Local(ComponentPose=@n4)) {layerGroup=None}
   Sequence Player 'backpack_motion'() {playRate=0.25, sequence=/Game/Accessories/FORT_Backpacks/F_MED_SushiChef_Backpack/Meshes/backpack_motion}
  ]}
 EventGraph {nodes [
   Event BlueprintUpdateAnimation()
   > n1 = Set float_var(float_var:double=0) -> (Output_Get:double)
   n2 = GetOwnerSkeletalMeshComponent()
   n3 = BreakRotator(InRot=BreakTransform(InTransform=GetDeltaTransformFromRefPose(self=@n2, BoneName=spine_05, BaseName=spine_01)).Rotation)
   Set backpack_rotate(execute=[@n4.OutputPin, @n1.then], backpack_rotate=MakeRotator(Roll=MapRangeClamped(Value=@n3.Yaw, InRangeA=-5, InRangeB=-30, OutRangeB=35)))
   > Branch(Condition=Get is_skydiving() -> (is_skydiving:bool))
   > Set float_var(float_var=1)
   > Branch(Condition=Get bIsSkydiveDiveMode())
   > Set float_var(float_var=0.45)
   Event OnBeginSkydiving()
   > Set is_skydiving(is_skydiving:bool=true) -> (Output_Get:bool)
   Event OnEndSkydiving()
   > Set is_skydiving(is_skydiving=false)
   n4 = Reroute Node()
  ]}
]
variables [
 backpack_rotate {type=Rotator}
 float_var {type=number}
 is_skydiving {type=boolean}
]
